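# Physics asset for the Bandolier Slurp morph head of a medium female player character, setting up bodies and constraints for dynamic hair.
# A head capsule anchors five small box base bodies through constraints with free swing and twist.
# Each base leads a hair chain: two right-side capsules, one large left-side slurp box, and two two-segment top slurp capsule chains.
# Constraints along the chains set tight swing and twist limits.
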PhysicsAsset /Game/Characters/Player/Female/Medium/Heads/F_MED_Bandolier_Slurp_Head/Meshes/F_MED_Bandolier_SlurpMorph_Head_Physics
bodies {
 dyn_hairA_01_Base_r [{shapeType=Box, extentX=1.01, extentY=1.01, extentZ=1.01, rotation=<rot 0 -90 0>}]
 dyn_hairA_01_r [{shapeType=Capsule, length=5.155, radius=1.1827, center=<3.0755 -0.2988 -0.0238>, rotation=<rot 0 -86.1603 -87.0556>}]
 dyn_hairB_01_Base_r [{shapeType=Box, extentX=1.01, extentY=1.01, extentZ=1.01, rotation=<rot 0 -90 0>}]
 dyn_hairB_01_r [{shapeType=Capsule, length=4.9215, radius=1.0996, center=<3.1115 0.2174 0.0157>, rotation=<rot 0 87.0217 -89.7818>}]
 dyn_hairSlurp_01_Base_l [{shapeType=Box, extentX=1.01, extentY=1.01, extentZ=1.01, rotation=<rot 0 -90 0>}]
 dyn_hairSlurp_01_l [{shapeType=Box, extentX=3.6792, extentY=9.3092, extentZ=16.1706, center=<3.7274 0 0>, rotation=<rot 0 90 0>}]
 dyn_hairTopSlurpA_01 [{shapeType=Capsule, length=1.234, radius=1.1526, center=<1 0 0>, rotation=<rot 90 0 0>}]
 dyn_hairTopSlurpA_01_Base [{shapeType=Box, extentX=1.01, extentY=1.01, extentZ=1.01, rotation=<rot 0 -90 0>}]
 dyn_hairTopSlurpA_02 [{shapeType=Capsule, length=1.4697, radius=0.7823, center=<1 0 0>, rotation=<rot 90 0 0>}]
 dyn_hairTopSlurpB_01 [{shapeType=Capsule, length=1.1147, radius=0.8158, center=<1 0 0>, rotation=<rot 90 0 0>}]
 dyn_hairTopSlurpB_01_Base [{shapeType=Box, extentX=1.01, extentY=1.01, extentZ=1.01, rotation=<rot 0 -90 0>}]
 dyn_hairTopSlurpB_02 [{shapeType=Capsule, length=1.0285, radius=0.65, center=<1 0 0>, rotation=<rot 90 0 0>}]
 head [{shapeType=Capsule, length=11.5599, radius=5.8421, center=<3.2009 1.3916 0>, rotation=<rot 90 0 0>}]
}
constraints [
 {
  bone1Name=dyn_hairSlurp_01_Base_l
  bone2Name=head
  swing1LimitDegrees=45
  swing1Motion=Free
  swing2LimitDegrees=45
  swing2Motion=Free
  twistLimitDegrees=45
  twistMotion=Free
 }
 {
  bone1Name=dyn_hairTopSlurpB_01_Base
  bone2Name=head
  swing1LimitDegrees=45
  swing1Motion=Free
  swing2LimitDegrees=45
  swing2Motion=Free
  twistLimitDegrees=45
  twistMotion=Free
 }
 {
  bone1Name=dyn_hairTopSlurpA_01_Base
  bone2Name=head
  swing1LimitDegrees=45
  swing1Motion=Free
  swing2LimitDegrees=45
  swing2Motion=Free
  twistLimitDegrees=45
  twistMotion=Free
 }
 {
  bone1Name=dyn_hairB_01_Base_r
  bone2Name=head
  swing1LimitDegrees=45
  swing1Motion=Free
  swing2LimitDegrees=45
  swing2Motion=Free
  twistLimitDegrees=45
  twistMotion=Free
 }
 {
  bone1Name=dyn_hairA_01_Base_r
  bone2Name=head
  swing1LimitDegrees=45
  swing1Motion=Free
  swing2LimitDegrees=45
  swing2Motion=Free
  twistLimitDegrees=45
  twistMotion=Free
 }
 {bone1Name=dyn_hairSlurp_01_l, bone2Name=dyn_hairSlurp_01_Base_l, swing1LimitDegrees=20, swing2LimitDegrees=20, twistLimitDegrees=8}
 {bone1Name=dyn_hairTopSlurpB_01, bone2Name=dyn_hairTopSlurpB_01_Base, swing1LimitDegrees=16, swing2LimitDegrees=16, twistLimitDegrees=8}
 {bone1Name=dyn_hairTopSlurpB_02, bone2Name=dyn_hairTopSlurpB_01, swing1LimitDegrees=16, swing2LimitDegrees=16, twistLimitDegrees=8}
 {bone1Name=dyn_hairTopSlurpA_01, bone2Name=dyn_hairTopSlurpA_01_Base, swing1LimitDegrees=16, swing2LimitDegrees=16, twistLimitDegrees=8}
 {bone1Name=dyn_hairTopSlurpA_02, bone2Name=dyn_hairTopSlurpA_01, swing1LimitDegrees=20, swing2LimitDegrees=16, twistLimitDegrees=8}
 {bone1Name=dyn_hairB_01_r, bone2Name=dyn_hairB_01_Base_r, swing1LimitDegrees=20, swing2LimitDegrees=20, twistLimitDegrees=20}
 {bone1Name=dyn_hairA_01_r, bone2Name=dyn_hairA_01_Base_r, swing1LimitDegrees=20, swing2LimitDegrees=20, twistLimitDegrees=20}
]
