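# Animation Blueprint for the face accessory part of the female Bandolier Slurp Monster character head.
# It copies the pose from the attached parent mesh, then adds secondary physics with a RigidBody node using an override physics asset.
# The simulation runs in root-bone space with custom gravity, scaled and clamped component motion.
# When character customization loads, it stores a dynamic material instance from the part mesh's slot 1 in PartMID.
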AnimBlueprint /Game/Characters/Player/Female/Medium/Heads/F_MED_Bandolier_Slurp_Head/Meshes/Parts/F_MED_BandolierSlurpMonster_FaceAcc_AnimBP
target_skeleton=&/F_MED_BandolierSlurpMonster_FaceAcc_Skeleton
graphs [
 AnimGraph {nodes [
   n1 = RigidBody(ComponentPose=Local To Component(LocalPose=Copy Pose From Mesh() {bUseAttachedParent=true})) {
    bClampLinearTranslationLimitToRefPose=true
    bOverrideWorldGravity=true
    overridePhysicsAsset=/Game/Characters/Player/Female/Medium/Heads/F_MED_Bandolier_Slurp_Head/Meshes/F_MED_Bandolier_SlurpMorph_Head_Physics
    simulationSpace=BaseBoneSpace
    alphaBoolBlend {blendOption=Linear}
    baseBoneRef {boneName=root}
    componentAppliedLinearAccClamp=<2500 2500 2500>
    componentLinearAccScale=<0.4 0.4 0.4>
    componentLinearVelScale=<0.25 0.25 0.25>
    overrideWorldGravity {z=-980}
   }
   Output Pose(Result=Component To Local(ComponentPose=@n1)) {layerGroup=None}
  ]}
 EventGraph {nodes [
   Event OnCharacterCustomizationLoaded()
   n1 = GetPartSkeletalMeshComponent()
   > Is Valid(InputObject=@n1)
   > n2 = Get Dynamic Material Instance(SkeletalMeshComponent=@n1, ElementIndex=1)
   > Set PartMID(PartMID:Material Instance Dynamic Object Reference=@n2) -> (Output_Get:Material Instance Dynamic Object Reference)
  ]}
]
variables []
